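# Translucent unlit master material for a shockwave refraction sphere, built for Cascade and Niagara mesh particles.
# Fresnel times particle alpha drives the index-of-refraction lerp, with a feature level switch substituting a constant on lower-end targets.
# Contrast-adjusted inverted Fresnel masks emissive, built from particle color times a desaturatable color parameter and passed through MF_FX_EmissiveAdjustments.
# The same mask times a dynamic particle parameter forms the opacity term.
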Material /DealershipSharedLibrary/FX/Materials/MasterMaterials/Shockwave/M_DLR_Refraction_Sphere
graph {
 properties {
  bUsedWithMeshParticles=true
  bUsedWithNiagaraMeshParticles=true
  blendMode=BLEND_Translucent
  pixelDepthOffsetMode=PDOM_Legacy
  refractionMethod=RM_IndexOfRefraction
  shadingModel=MSM_Unlit
  usedShadingModels=Unlit
 }
 nodes [
  n1 = Constant()
  n2 = MaterialFunctionCall(Exponent=ScalarParameter() {defaultValue=2, parameterName=Sphere Fresnel Exponent}) {materialFunction=&MF_Fresnel_FX}
  LinearInterpolate(
   B=FeatureLevelSwitch(Default=ScalarParameter() {defaultValue=0.96, parameterName=Refraction Amount}, ES2=@n1, ES3_1=@n1, SM4_REMOVED=@n1)
   Alpha=Multiply(A=@n2.Fresnel, B=ParticleColor().A)
  ) {constA=1, constB=0.96}
  n3 = ParticleColor()
  n4 = Desaturation(
   None=VectorParameter() {parameterName=Emissive Color, defaultValue=<rgba 1 1 1 1>}.RGB
   Fraction=ScalarParameter() {defaultValue=1, parameterName=Use Particle Color for Emissive?}
  )
  n5 = Power(Base=@n2.Fresnel Inverted, Exp=ScalarParameter() {defaultValue=1, parameterName=Emissive Fresnel Contrast})
  MaterialFunctionCall(Emissive=Multiply(A=Multiply(A=@n3.RGB, B=@n4), B=@n5)) {materialFunction=&MF_FX_EmissiveAdjustments}
  Multiply(A=@n5, B=DynamicParameter() {defaultValue=<rgba 0 1 1 1>, paramNames [Base Color Opacty, Param2, Param3, Param4]}.Base Color Opacty)
 ]
}
