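# Translucent unlit material for energy shockwave effects on Niagara mesh particles, colored by particle color through an emissive adjustment function.
# Opacity combines an invertible Fresnel base and edge falloff with optional energy noise, then passes through depth fade.
# View-space surface normals supply matcap-style UVs for two counter-panning flow samples, whose transformed sum becomes the energy texture's UVs.
# Particle alpha is subtracted from that noise, so higher alpha erodes the effect.
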Material /DealershipSharedLibrary/FX/Materials/MasterMaterials/Shockwave/M_EnergyShockwave
graph {
 properties {bUsedWithNanite=true, bUsedWithNiagaraMeshParticles=true, blendMode=BLEND_Translucent, shadingModel=MSM_Unlit, usedShadingModels=Unlit}
 nodes [
  n1 = MaterialFunctionCall(
   Invert Fresnel=StaticBoolParameter() {defaultValue=true, parameterName=Invert Fresnel}
   Power=ScalarParameter() {defaultValue=1, parameterName=Fresnel Power}
  ) {materialFunction=&Fresnel_Function}
  n2 = Power(Base=@n1.Result, Exp=ScalarParameter() {defaultValue=4, parameterName=Fresnel Base Power}) {constExponent=4}
  n3 = ScalarParameter() {defaultValue=64, parameterName=Fresnel Base Amount}
  n4 = Power(Base=OneMinus(None=@n1.Result)) {constExponent=8}
  n5 = ScalarParameter() {defaultValue=1, parameterName=Fresnel Edge Amount}
  n6 = Saturate(None=Multiply(A=Multiply(A=@n2, B=@n3) {constB=64}, B=LinearInterpolate(B=@n4, Alpha=@n5) {constA=1}))
  n7 = Constant3Vector() {constant=<rgba 0 0 1 0>}
  n8 = ConstantBiasScale(None=ComponentMask(None=Transform(None=@n7) {transformType=TRANSFORM_View}) {g=true, r=true}) {bias=0.5}
  n9 = Add(
   A=TextureSample(UVs=Panner(Coordinate=@n8) {speedX=0.05, speedY=0.2}) {texture=/DealershipSharedLibrary/FX/Textures/Noises/T_DLR_Flow03_D}.RGB
   B=TextureSample(UVs=Panner(Coordinate=@n8) {speedX=0.1, speedY=-0.3}) {texture=&T_DLR_Flow03_D}.RGB
  )
  n10 = ConstantBiasScale(None=ComponentMask(None=Transform(None=@n9) {transformSourceType=TRANSFORMSOURCE_View, transformType=TRANSFORM_Camera}) {g=true, r=true}) {bias=0.5}
  n11 = ParticleColor()
  n12 = StaticSwitchParameter(True=Subtract(A=TextureSample(UVs=@n10) {texture=/DealershipSharedLibrary/FX/Textures/Noises/T_Energy}.G, B=@n11.A), False=Constant() {r=1}) {defaultValue=true, parameterName=Add Noise?}
  MaterialFunctionCall(
   Opacity In=Multiply(A=Multiply(A=@n6, B=@n12), B=ScalarParameter() {defaultValue=1, parameterName=Opacity})
   DepthFade Distance=ScalarParameter() {defaultValue=100, parameterName=DepthFade Distance}
  ) {materialFunction=&MF_VCVFX_DepthFade}
  MaterialFunctionCall(Emissive=@n11.RGB) {materialFunction=&MF_FX_EmissiveAdjustments}
 ]
}
