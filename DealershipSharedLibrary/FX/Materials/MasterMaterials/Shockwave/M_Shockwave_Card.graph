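# Translucent master material for shockwave effects on Niagara sprite and mesh particles, drawing a procedural radial shape on a card.
# One or two panning textures in linear or curved (twisted) radial UVs supply noise that also distorts a sphere mask.
# An interior mask shapes the center and adds an edge highlight, with dynamic parameters for noise and interior masking.
# Opacity gets contrast, depth fade and particle alpha, while particle color tints base and emissive color.
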Material /DealershipSharedLibrary/FX/Materials/MasterMaterials/Shockwave/M_Shockwave_Card
graph {
 properties {
  bAutomaticallySetUsageInEditor=false
  bUsedWithNiagaraMeshParticles=true
  bUsedWithNiagaraSprites=true
  blendMode=BLEND_Translucent
  usedShadingModels=Default Lit
 }
 nodes [
  MaterialFunctionCall(Emissive=NamedRerouteUsage()) {materialFunction=&MF_FX_EmissiveAdjustments}
  MaterialFunctionCall(Opacity In=NamedRerouteUsage(), DepthFade Distance=ScalarParameter() {group=Global Parameters, parameterName=Depth Fade}) {materialFunction=&MF_VCVFX_DepthFade}
  n1 = NamedRerouteUsage()
  n2 = AppendVector(
   A=ScalarParameter() {defaultValue=2, group=Texture 2, parameterName=Tex2 - Radial Tile X}
   B=Multiply(A=ScalarParameter() {defaultValue=2, group=Texture 2, parameterName=Tex2 - Radial Tile Y}) {constB=0.25}
  )
  n3 = MaterialFunctionCall(Vector or UVs=Add(A=TextureCoordinate(), B=Constant() {r=-0.5})) {materialFunction=&VectorToRadialValue}
  n4 = Add(A=@n3.Radial Coordinates, B=@n3.Radial Coordinates)
  n5 = ScalarParameter() {defaultValue=-0.05, group=Texture 2, parameterName=Tex2 - Radial Panning Speed X}
  n6 = ScalarParameter() {defaultValue=-0.05, group=Texture 2, parameterName=Tex2 - Radial Panning Speed Y}
  n7 = TextureObjectParameter() {
   group=Texture 2
   parameterName=Tex2 - Radial Texture
   samplerType=SAMPLERTYPE_Grayscale
   texture=/Game/Effects/Fort_Effects/Textures/Tile/T_Concrete_Crack_A
  }
  n8 = TextureSample(UVs=Panner(Coordinate=Multiply(A=@n4, B=@n2), Speed=AppendVector(A=@n5, B=@n6)), Tex=@n7) {samplerType=SAMPLERTYPE_Masks, texture=/Game/Effects/Fort_Effects/Textures/T_PackedNoise}
  n9 = ChannelMaskParameter(None=@n8.RGBA) {group=Texture 2, parameterName=Texture 2 - Channel, sortPriority=11}
  n10 = ScalarParameter() {group=Texture 2, parameterName=Tex 2 - Use Channel}
  n11 = Add(
   A=Multiply(A=LinearInterpolate(A=@n9.R, B=@n8.RGB, Alpha=@n10), B=ScalarParameter() {defaultValue=1, group=Texture 2, parameterName=Tex2 - Amount})
   B=ScalarParameter() {group=Texture 2, parameterName=Tex2 - Bias}
  )
  NamedRerouteDeclaration(None=StaticSwitchParameter(True=Add(A=@n1, B=@n11), False=NamedRerouteUsage()) {defaultValue=true, group=Global Parameters, parameterName=2 Texture Layers}) {name=2 Layers Linear Radial Texture, nodeColor=<rgba 0.7574 0.2584 1 1>}
  n12 = StaticSwitchParameter(True=NamedRerouteUsage(), False=Constant() {r=1}) {defaultValue=true, group=Global Parameters, parameterName=Mask Out Center?}
  n13 = ScalarParameter() {defaultValue=1, group=Radial Texture, parameterName=Radial Texture Amount, sortPriority=1}
  n14 = NamedRerouteUsage()
  n15 = Add(
   A=Power(Base=NamedRerouteUsage(), Exp=Multiply(A=@n13, B=@n14))
   B=ScalarParameter() {group=Radial Texture, parameterName=Radial Texture Bias, sortPriority=1}
  )
  n16 = Add(
   A=Multiply(A=@n12, B=ScalarParameter() {defaultValue=0.1, group=Sphere Mask, parameterName=Mask Bias})
   B=Multiply(A=@n12, B=Saturate(None=@n15))
  )
  n17 = ScalarParameter() {defaultValue=1, group=Global Parameters, parameterName=Texture Power}
  n18 = MaterialFunctionCall(In=Saturate(None=Power(Base=@n16, Exp=@n17)), Contrast=ScalarParameter() {group=Global Parameters, parameterName=Texture Contrast}) {materialFunction=&CheapContrast}
  NamedRerouteDeclaration(None=Multiply(
    A=ParticleColor().A
    B=MaterialFunctionCall(Opacity In=@n18.Result) {materialFunction=&MF_VCVFX_DepthFade}.Result
   )) {name=Opacity Out, nodeColor=<rgba 0.0479 1 0.9905 1>}
  n19 = DynamicParameter() {defaultValue=<rgba 1 1 0 1>, paramNames [Texture Noise Amount, Interior Mask Amount, Interior Mask Bias, Param4]}
  NamedRerouteDeclaration(None=@n19.Interior Mask Amount) {name=DP_Interior Mask Amount, nodeColor=<rgba 0.5 1 0 1>}
  n20 = MaterialFunctionCall!(
   UV In=TextureCoordinate()
   Tile X=ScalarParameter() {defaultValue=2, group=Texture 2, parameterName=Tex2 - Radial Tile X}
   Tile Y=ScalarParameter() {defaultValue=2, group=Texture 2, parameterName=Tex2 - Radial Tile Y}
   Panning Speed X=ScalarParameter() {defaultValue=-0.05, group=Texture 2, parameterName=Tex2 - Radial Panning Speed X}
   Panning Speed Y=ScalarParameter() {defaultValue=-0.05, group=Texture 2, parameterName=Tex2 - Radial Panning Speed Y}
   Twist=ScalarParameter() {defaultValue=1, group=Texture 2, parameterName=Tex2 - Radial Center Power}
   Exp X=ScalarParameter() {defaultValue=0.5, group=Texture 2, parameterName=Tex2 - Radial Number of Turns}
  ) {materialFunction=/DealershipSharedLibrary/FX/Materials/MaterialFunctions/UVs/MF_DLR_RadialUV}
  n21 = TextureObjectParameter() {
   group=Texture 2
   parameterName=Tex2 - Radial Texture
   samplerType=SAMPLERTYPE_Grayscale
   texture=&T_Concrete_Crack_A
  }
  n22 = TextureSample!(UVs=@n20.Input UVs, Tex=@n21, "DDX(UVs)"=@n20.Input UVs, "DDY(UVs)"=@n20.Input UVs, Apply View MipBias) {mipValueMode=TMVM_Derivative}
  n23 = ChannelMaskParameter(None=@n22.RGBA) {group=Texture 2, parameterName=Texture 2 - Channel, sortPriority=11}
  n24 = ScalarParameter() {group=Texture 2, parameterName=Tex 2 - Use Channel}
  n25 = Add(
   A=Multiply(A=LinearInterpolate(A=@n22.RGB, B=@n23.R, Alpha=@n24), B=ScalarParameter() {defaultValue=1, group=Texture 2, parameterName=Tex2 - Amount})
   B=ScalarParameter() {group=Texture 2, parameterName=Tex2 - Bias}
  )
  n26 = NamedRerouteUsage()
  NamedRerouteDeclaration(None=StaticSwitchParameter(True=Add(A=@n25, B=@n26), False=NamedRerouteUsage()) {defaultValue=true, group=Global Parameters, parameterName=2 Texture Layers}) {name=2 Layers Curved Radial Texture, nodeColor=<rgba 0.191 0.1321 1 1>}
  NamedRerouteDeclaration(None=@n19.Interior Mask Bias) {name=DP_Interior Mask Bias, nodeColor=<rgba 0.5 1 0 1>}
  n27 = ParticleColor()
  n28 = ScalarParameter() {defaultValue=1, group=Global Parameters, parameterName=Use Particle Color?}
  NamedRerouteDeclaration(None=Multiply(
    A=VectorParameter() {group=Global Parameters, parameterName=Base Color, sortPriority=64, defaultValue=<rgba 0 0 0 1>}.RGB
    B=Desaturation(None=@n27.RGB, Fraction=@n28)
   )) {name=Base Color Out, nodeColor=<rgba 0.9913 1 0.1871 1>}
  n29 = VectorParameter() {group=Global Parameters, parameterName=Emissive Color, sortPriority=64, defaultValue=<rgba 1 1 1 1>}
  n30 = Desaturation(None=ParticleColor().RGB, Fraction=ScalarParameter() {defaultValue=1, group=Global Parameters, parameterName=Use Particle Color?})
  NamedRerouteDeclaration(None=Multiply(A=Multiply(A=@n29.RGB, B=@n30), B=ScalarParameter() {defaultValue=1, group=Global Parameters, parameterName=Emissive Boost})) {name=Emissive Out, nodeColor=<rgba 1 0.1372 0.7978 1>}
  n31 = NamedRerouteUsage()
  n32 = NamedRerouteUsage()
  n33 = Multiply(
   A=ComponentMask(None=StaticSwitchParameter(True=@n31, False=@n32) {defaultValue=true, group=Global Parameters, parameterName=Curved Radial Texture}) {g=true, r=true}
   B=Multiply(A=ScalarParameter() {defaultValue=0.075, group=Sphere Mask, parameterName=Sphere Mask Noise Amount}, B=NamedRerouteUsage())
  ) {constB=0.1}
  n34 = TextureCoordinate()
  NamedRerouteDeclaration(None=SphereMask(
    A=Add(A=@n33, B=@n34)
    B=Constant() {r=0.5}
    Radius=ScalarParameter() {defaultValue=0.25, group=Sphere Mask, parameterName=Sphere Mask Radius}
    Hardness=ScalarParameter() {defaultValue=0.5, group=Sphere Mask, parameterName=Sphere Mask Hardness}
   )) {name=Base Texture Out, nodeColor=<rgba 0.2011 1 0.3829 1>}
  n35 = NamedRerouteUsage()
  n36 = OneMinus(None=NamedRerouteUsage())
  n37 = Multiply(A=ScalarParameter() {defaultValue=2, group=Interior Mask, parameterName=Interior Mask Amount}, B=NamedRerouteUsage())
  n38 = Add(
   A=Saturate(None=Power(Base=@n36, Exp=@n37))
   B=Add(A=ScalarParameter() {group=Interior Mask, parameterName=Interior Mask Bias}, B=NamedRerouteUsage())
  )
  n39 = ScalarParameter() {defaultValue=2, group=Interior Mask, parameterName=Interior Mask Amount}
  n40 = NamedRerouteUsage()
  n41 = NamedRerouteUsage()
  n42 = Power(
   Base=OneMinus(None=@n41)
   Exp=Multiply(A=Multiply(A=@n39, B=@n40), B=ScalarParameter() {defaultValue=7, group=Edge Highlight, parameterName=Edge Highlight Thickness}) {constB=10}
  )
  n43 = Multiply(A=@n41, B=Add(A=Saturate(None=@n42), B=ScalarParameter() {group=Interior Mask, parameterName=Interior Mask Bias}))
  n44 = ScalarParameter() {defaultValue=20, group=Edge Highlight, parameterName=Edge Highlight Amount}
  NamedRerouteDeclaration(None=Add(A=Multiply(A=@n35, B=@n38), B=Multiply(A=@n43, B=@n44) {constB=10})) {name=Interior Mask, nodeColor=<rgba 1 0.6498 0.1918 1>}
  n45 = AppendVector(
   A=ScalarParameter() {defaultValue=2, group=Texture 1, parameterName=Tex1 - Radial Tile X}
   B=Multiply(A=ScalarParameter() {defaultValue=2, group=Texture 1, parameterName=Tex1 - Radial Tile Y}) {constB=0.25}
  )
  n46 = ScalarParameter() {defaultValue=-0.05, group=Texture 1, parameterName=Tex1 - Radial Panning Speed X}
  n47 = ScalarParameter() {defaultValue=-0.05, group=Texture 1, parameterName=Tex1 - Radial Panning Speed Y}
  n48 = TextureSample!(
   UVs=Panner(Coordinate=Multiply(A=@n45, B=@n4), Speed=AppendVector(A=@n46, B=@n47))
   Tex=TextureObjectParameter() {group=Texture 1, parameterName=Tex1 - Radial Texture, texture=/Game/Effects/Fort_Effects/Textures/Tile/T_GaussianNoise_01_M}
   Apply View MipBias
  ) {samplerType=SAMPLERTYPE_Masks, texture=&T_PackedNoise}
  NamedRerouteDeclaration(None=LinearInterpolate(
    A=ChannelMaskParameter(None=@n48.RGBA) {group=Texture 1, parameterName=Tex1 - Channel, sortPriority=11}.R
    B=@n48.RGB
    Alpha=ScalarParameter() {group=Texture 1, parameterName=Tex1 - Use Channel}
   )) {name=Linear Radial Texture, nodeColor=<rgba 1 0 0.5859 1>}
  n49 = MaterialFunctionCall(
   UV In=TextureCoordinate()
   Tile X=ScalarParameter() {defaultValue=2, group=Texture 1, parameterName=Tex1 - Radial Tile X}
   Tile Y=ScalarParameter() {defaultValue=2, group=Texture 1, parameterName=Tex1 - Radial Tile Y}
   Panning Speed X=ScalarParameter() {defaultValue=-0.05, group=Texture 1, parameterName=Tex1 - Radial Panning Speed X}
   Panning Speed Y=ScalarParameter() {defaultValue=-0.05, group=Texture 1, parameterName=Tex1 - Radial Panning Speed Y}
   Twist=ScalarParameter() {defaultValue=1, group=Texture 1, parameterName=Tex1 - Radial Center Power}
   Exp X=ScalarParameter() {defaultValue=0.5, group=Texture 1, parameterName=Tex1 - Radial Number of Turns}
  ) {materialFunction=&MF_DLR_RadialUV}
  n50 = TextureSample!(
   UVs=@n49.Input UVs
   Tex=TextureObjectParameter() {group=Texture 1, parameterName=Tex1 - Radial Texture, texture=&T_GaussianNoise_01_M}
   "DDX(UVs)"=@n49.Input UVs
   "DDY(UVs)"=@n49.Input UVs
   Apply View MipBias
  ) {mipValueMode=TMVM_Derivative}
  NamedRerouteDeclaration(None=LinearInterpolate(
    A=@n50.RGB
    B=ChannelMaskParameter(None=@n50.RGBA) {group=Texture 1, parameterName=Tex1 - Channel, sortPriority=11}.R
    Alpha=ScalarParameter() {group=Texture 1, parameterName=Tex1 - Use Channel}
   )) {name=Curved Radial Texture, nodeColor=<rgba 0 0.9062 1 1>}
  NamedRerouteDeclaration(None=StaticSwitchParameter(True=NamedRerouteUsage(), False=NamedRerouteUsage()) {defaultValue=true, group=Global Parameters, parameterName=Curved Radial Texture}) {name=Radial Texture Mask, nodeColor=<rgba 0.1719 1 0 1>}
  NamedRerouteDeclaration(None=@n19.Texture Noise Amount) {name=DP_Texture Noise Amount, nodeColor=<rgba 0 0.1797 1 1>}
  NamedRerouteUsage()
 ]
}
